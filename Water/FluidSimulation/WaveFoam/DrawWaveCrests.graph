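# Unlit material that draws wave-crest data over the water capture area: foam in R, shallow-water depth as beach wet mask in G, wave height in B.
# Chained Gerstner wave evaluations iteratively undo horizontal displacement, finding the wave at each texel's world position.
# Crest foam combines normal-Z and height thresholds, adds a box-masked FoamVolume texture sample, then fades at the capture edges.
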Material /Water/FluidSimulation/WaveFoam/DrawWaveCrests
graph {
 properties {bAllowNegativeEmissiveColor=true, pixelDepthOffsetMode=PDOM_Legacy, shadingModel=MSM_Unlit, usedShadingModels=Unlit}
 nodes [
  n1 = TextureSampleParameter2D() {
   desc="R=Water Depth, G=Terrain Elevation"
   parameterName=ShallowWaterRT
   samplerType=SAMPLERTYPE_LinearColor
   texture=/Engine/Functions/Engine_MaterialFunctions02/PivotPainter2/Black_1x1_EXR_Texture
  }
  Add(A=@n1.R)
  n2 = ScalarParameter() {defaultValue=0.2, parameterName=EdgeMask}
  n3 = MaterialFunctionCall!(A=TextureCoordinate(), B=Constant() {r=0.5}, Bounds=OneMinus(None=Multiply(A=@n2) {constB=2}), Edge Falloff=@n2) {materialFunction=/Engine/Functions/Engine_MaterialFunctions01/Gradient/BoxMask-2D}
  n4 = TextureCoordinate()
  n5 = Add(
   A=Multiply(A=Subtract(A=@n4) {constB=0.5}, B=CollectionParameter() {collection=/Water/Materials/MPC/MPC_Water, parameterName=WaveCaptureSize})
   B=ComponentMask(None=CollectionParameter() {collection=&MPC_Water, parameterName=SimLocation}) {g=true, r=true}
  )
  n6 = Add(
   A=MaterialFunctionCall() {materialFunction=/Water/Materials/Functions/GetWaterTime}.WaterTime
   B=ScalarParameter() {parameterName=FoamTimeBias}
  )
  n7 = AppendVector(A=@n5, B=Constant())
  n8 = MaterialFunctionCall!(
   WaveParameterRT
   Override Position=Subtract(A=@n7, B=MaterialFunctionCall(Override Position=@n5, Time=@n6) {materialFunction=/Water/Caustics/Materials/Functions/GetGerstnerWaves}.WPO)
   Time=@n6
  ) {materialFunction=&GetGerstnerWaves}
  n9 = Subtract(
   A=@n7
   B=MaterialFunctionCall(Override Position=Subtract(A=@n7, B=@n8.WPO), Time=@n6) {materialFunction=&GetGerstnerWaves}.WPO
  )
  n10 = MaterialFunctionCall(
   Override Position=Subtract(A=@n7, B=MaterialFunctionCall(Override Position=@n9, Time=@n6) {materialFunction=&GetGerstnerWaves}.WPO)
   Time=@n6
  ) {materialFunction=&GetGerstnerWaves}
  n11 = Subtract(A=Multiply(A=ComponentMask(None=@n10.Normal) {b=true}) {constB=-1}, B=ScalarParameter() {parameterName=Bias})
  n12 = ComponentMask(None=@n10.WPO) {b=true}
  n13 = Subtract(A=@n12, B=ScalarParameter() {defaultValue=48, parameterName=Min WPO Foam})
  n14 = ScalarParameter() {defaultValue=48, parameterName=Foam WPo Divisor}
  n15 = Add(
   A=Multiply(A=Saturate(None=@n11), B=ScalarParameter() {defaultValue=10, parameterName=Normal Amount})
   B=Max(A=Divide(A=@n13, B=@n14)) {constB=0}
  )
  n16 = Subtract(A=TextureCoordinate()) {constB=0.5}
  n17 = CollectionParameter() {collection=&MPC_Water, parameterName=WaveCaptureSize}
  n18 = VectorParameter() {parameterName=FoamVolumePosition}
  n19 = Divide(
   A=Subtract(A=Multiply(A=@n16, B=@n17) {desc=Foam World Position}, B=ComponentMask(None=@n18.RGB) {g=true, r=true})
   B=ScalarParameter() {defaultValue=2048, parameterName=FoamVolumeSize}
  )
  n20 = TextureObjectParameter() {
   parameterName=FoamVolume
   samplerType=SAMPLERTYPE_LinearColor
   texture=&Black_1x1_EXR_Texture
  }
  n21 = TextureSample(UVs=Saturate(None=Add(A=@n19) {constB=0.5}), Tex=@n20)
  n22 = ScalarParameter() {parameterName=Foam Volume Additions}
  n23 = Multiply(
   A=Subtract(A=TextureCoordinate()) {constB=0.5}
   B=CollectionParameter() {collection=&MPC_Water, parameterName=WaveCaptureSize}
  ) {desc=Foam World Position}
  n24 = ComponentMask(None=VectorParameter() {parameterName=FoamVolumePosition}.RGB) {g=true, r=true}
  n25 = Divide(
   A=AppendVector(A=Subtract(A=@n23, B=@n24), B=ScalarParameter() {defaultValue=16, parameterName=Z})
   B=ScalarParameter() {defaultValue=2048, parameterName=FoamVolumeSize}
  )
  n26 = ComponentMask(None=Saturate(None=Add(A=@n25) {constB=0.5})) {g=true, r=true}
  n27 = Constant() {r=0.5}
  n28 = ScalarParameter() {defaultValue=0.2, parameterName=EdgeMask}
  n29 = OneMinus(None=Multiply(A=@n28) {constB=2})
  n30 = Multiply(
   A=Multiply(A=Multiply(A=@n21.R, B=@n22))
   B=SquareRoot(None=MaterialFunctionCall(A=@n26, B=@n27, Bounds=@n29, Edge Falloff=@n28) {materialFunction=&BoxMask-2D}.Result)
  )
  MaterialFunctionCall(X=Multiply(A=SquareRoot(None=@n3.Result), B=Add(A=@n15, B=@n30)), Y=@n1.R, Z=@n12) {desc="R=Foam, G=Beach Wet Mask, B=Wave Height", materialFunction=&MakeFloat3}
 ]
}
