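# Unlit material in the fluid simulation's wave foam set that, going by its name, copies foam delta data from a render target.
# It samples the CurrentFrame texture (FoamRT by default), scales red by parameter D and remaps it through a custom 1-exp(-x) saturation curve.
# The result is recombined with the unchanged green and blue channels, alongside a separate red-green mask branch.
# Negative emissive values are allowed.
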Material /Water/FluidSimulation/WaveFoam/FoamDeltaCopy
graph {
 properties {bAllowNegativeEmissiveColor=true, pixelDepthOffsetMode=PDOM_Legacy, shadingModel=MSM_Unlit, usedShadingModels=Unlit}
 nodes [
  n1 = TextureSampleParameter2D() {parameterName=CurrentFrame, samplerType=SAMPLERTYPE_LinearColor, texture=&/FoamRT}
  MaterialExpressionMultiply_0 = Multiply(A=@n1.R, B=ScalarParameter() {defaultValue=0.1, parameterName=D})
  n2 = Custom!(x=@MaterialExpressionMultiply_0) {
   code="return 1.0-exp(-x);"
   description=1-exp
   outputType=CMOT_Float1
   inputs [{inputName=x, input {expression=&self:MaterialExpressionMultiply_0}}]
  }
  AppendVector(A=AppendVector(A=@n2, B=@n1.G), B=@n1.B)
  ComponentMask(None=@n1.RGB) {g=true, r=true}
 ]
}
