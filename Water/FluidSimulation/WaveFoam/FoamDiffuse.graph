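# Unlit simulation material that advances wave foam and wetness from the previous render target, likely for the water fluid simulation.
# Foam in RT is reprojected by the simulation location's movement, blurred with a five-tap cross, faded, added to WaveRT red and masked near the borders.
# Wetness accumulates from WaveRT green with slow decay, sampled without reprojection.
# Foam, saturated wetness and WaveRT blue are packed into RGB.
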Material /Water/FluidSimulation/WaveFoam/FoamDiffuse
graph {
 properties {bAllowNegativeEmissiveColor=true, pixelDepthOffsetMode=PDOM_Legacy, shadingModel=MSM_Unlit, usedShadingModels=Unlit}
 nodes [
  n1 = ScalarParameter() {defaultValue=0.2, parameterName=asdf}
  n2 = Multiply(A=@n1) {constB=2}
  n3 = SquareRoot(None=MaterialFunctionCall!(A=TextureCoordinate(), B=Constant() {r=0.5}, Bounds=OneMinus(None=@n2), Edge Falloff=@n1) {materialFunction=/Engine/Functions/Engine_MaterialFunctions01/Gradient/BoxMask-2D}.Result)
  n4 = SquareRoot(None=SquareRoot(None=SquareRoot(None=@n3)))
  n5 = Subtract(
   A=CollectionParameter() {collection=/Water/Materials/MPC/MPC_Water, parameterName=SimLocation1FrameBehind}
   B=CollectionParameter() {collection=&MPC_Water, parameterName=SimLocation}
  )
  n6 = Subtract(
   A=TextureCoordinate()
   B=Divide(
    A=ComponentMask(None=@n5) {g=true, r=true}
    B=CollectionParameter() {collection=&MPC_Water, parameterName=WaveCaptureSize}
   )
  )
  n7 = TextureSampleParameter2D(UVs=@n6) {
   parameterName=RT
   samplerType=SAMPLERTYPE_LinearColor
   texture=/Engine/Functions/Engine_MaterialFunctions02/PivotPainter2/Black_1x1_EXR_Texture
  }
  n8 = MaterialFunctionCall!(OffsetAmount=Divide(B=ScalarParameter() {defaultValue=512, parameterName=Res}) {constB=256}) {materialFunction=/Engine/Functions/Engine_MaterialFunctions01/ImageAdjustment/BlurSampleOffsets}
  n9 = TextureSampleParameter2D(UVs=Add(A=@n6, B=@n8."0,1")) {
   parameterName=RT
   samplerType=SAMPLERTYPE_LinearColor
   texture=&Black_1x1_EXR_Texture
  }
  n10 = TextureSampleParameter2D(UVs=Add(A=@n6, B=@n8."-1,0")) {
   parameterName=RT
   samplerType=SAMPLERTYPE_LinearColor
   texture=&Black_1x1_EXR_Texture
  }
  n11 = TextureSampleParameter2D(UVs=Add(A=@n6, B=@n8."1,0")) {
   parameterName=RT
   samplerType=SAMPLERTYPE_LinearColor
   texture=&Black_1x1_EXR_Texture
  }
  n12 = TextureSampleParameter2D(UVs=Add(A=@n6, B=@n8."0,-1")) {
   parameterName=RT
   samplerType=SAMPLERTYPE_LinearColor
   texture=&Black_1x1_EXR_Texture
  }
  n13 = Add(A=@n7.R, B=Add(A=Add(A=@n9.R, B=@n10.R), B=Add(A=@n11.R, B=@n12.R)))
  n14 = Multiply(
   A=LinearInterpolate(A=@n7.R, B=Divide(A=@n13) {constB=5}, Alpha=ScalarParameter() {defaultValue=0.5, parameterName=Diffuse})
   B=OneMinus(None=ScalarParameter() {parameterName=Fade})
  )
  n15 = TextureSampleParameter2D() {
   parameterName=WaveRT
   samplerType=SAMPLERTYPE_LinearColor
   texture=&Black_1x1_EXR_Texture
  }
  n16 = Add(A=@n14, B=@n15.R)
  n17 = TextureSampleParameter2D() {
   parameterName=RT
   samplerType=SAMPLERTYPE_LinearColor
   texture=&Black_1x1_EXR_Texture
  }
  n18 = ScalarParameter() {defaultValue=0.001, parameterName=Wet Fade}
  n19 = Add(A=Multiply(A=@n15.G, B=ScalarParameter() {defaultValue=0.02, parameterName=Wet per Frame}), B=Multiply(A=@n17.G, B=OneMinus(None=@n18)))
  AppendVector(A=AppendVector(A=Multiply(A=@n4, B=@n16), B=Saturate(None=@n19)), B=@n15.B)
  ComponentMask(None=@n14) {g=true, r=true}
 ]
}
