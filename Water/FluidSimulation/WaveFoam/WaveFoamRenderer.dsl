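; Wave foam renderer Blueprint that simulates ocean wave foam into the FoamRT render target, with start and stop events for editor ticking.
; Initialization creates two ping-pong foam targets, a wave target and three dynamic materials.
; Each tick draws wave crests, feeds them into a fading foam diffuse pass, then writes FoamRT from current and previous frames.
; Construction pushes freeze, time and capture size to the water parameter collection.
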(event Custom|Start%20Tick
  (Default|SetEditorTickEnabled true)
  (CallFunction|Initialize)
  (Rendering|Material|SetScalarParameterValue "/Water/Materials/MPC/MPC_Water.MPC_Water" "FreezeTime" :!class /Script/Engine.KismetMaterialLibrary)
  (Variables|Default|SetTicking true))

(event Custom|Stop%20Tick
  (Default|SetEditorTickEnabled false))

(event Custom|Allocate%20RTs
  ;[1 1 1] Initialize
  (Rendering|CreateAndAssignRenderTarget (Variables|Default|GetFoam%20RT%20A) (Variables|Default|GetRes) (Variables|Default|GetRes) "RTF_RG16f"
    (:Out
      (Rendering|CreateAndAssignRenderTarget (Variables|Default|GetFoam%20RT%20B) (Variables|Default|GetRes) (Variables|Default|GetRes) "RTF_RG16f"
        (:Out
          (Rendering|CreateAndAssignRenderTarget (Variables|Default|GetWave%20RT) (Variables|Default|GetRes) (Variables|Default|GetRes) "RTF_RG16f"))))))

(event Custom|Initialize
  ;[1 1 1] Initialize
  (CallFunction|AllocateRTs)
  (CallFunction|CreateMID))

(event Custom|Create%20MID
  ;[1 1 1] Initialize
  (Rendering|CreateAndAssignMID (Variables|Default|GetDrawWave%20MID) "/Water/FluidSimulation/WaveFoam/DrawWaveCrests.DrawWaveCrests" "WaveMID"
    (:Out
      (Rendering|CreateAndAssignMID (Variables|Default|GetFoamDiffuse%20MID) "/Water/FluidSimulation/WaveFoam/FoamDiffuse.FoamDiffuse" "FoamDiffuseMID"
        (:Out
          (Rendering|CreateAndAssignMID (Variables|Default|GetFoamToneMapMID) "/Water/FluidSimulation/WaveFoam/FoamDeltaCopy.FoamDeltaCopy" "FoamDeltaMID"))))))

(event Tick|EventCustomTick
  (params (DeltaSeconds :type Float%20%28single-precision%29))
  (bind (_read _write) (Utilities|PingPong))
  (bind (_read_2 _write_2) (Utilities|PingPong))
  (bind _foamtonemapmid (Variables|Default|GetFoamToneMapMID))
  (bind _foamdiffuse_mid (Variables|Default|GetFoamDiffuse%20MID))
  (Rendering|Material|SetScalarParameterValue (Variables|Default|GetFoamDiffuse%20MID) "Fade" (* (Variables|Default|GetFade) 0.1) :!class /Script/Engine.MaterialInstanceDynamic)
  (Rendering|Material|SetScalarParameterValue (Variables|Default|GetDrawWave%20MID) "FoamTimeBias" (Variables|Default|GetTime%20Bias) :!class /Script/Engine.MaterialInstanceDynamic)
  ;[0.4225 0.6232 1] Draw Waves First
  (Rendering|DrawMaterialtoRenderTarget (Variables|Default|GetWave%20RT) (Variables|Default|GetDrawWave%20MID))
  (Rendering|Material|SetTextureParameterValue _foamdiffuse_mid "RT" (Utilities|PingPong))
  (Rendering|Material|SetTextureParameterValue _foamdiffuse_mid "WaveRT" (Variables|Default|GetWave%20RT))
  (Rendering|DrawMaterialtoRenderTarget _write (Variables|Default|GetFoamDiffuse%20MID))
  (Rendering|Material|SetTextureParameterValue _foamtonemapmid "CurrentFrame" _write_2)
  (Rendering|Material|SetTextureParameterValue _foamtonemapmid "PreviousFrame" _read_2)
  (Rendering|DrawMaterialtoRenderTarget "/Water/FluidSimulation/WaveFoam/FoamRT.FoamRT" (Variables|Default|GetFoamToneMapMID))
  (Math|Integer|IncrementInt (Variables|Default|Getidx)))

(fn ConstructionScript
  (Rendering|Material|SetScalarParameterValue "/Water/Materials/MPC/MPC_Water.MPC_Water" "FreezeTime" (Math|Conversions|ToFloat(Boolean) (Variables|Default|GetFreeze)) :!class /Script/Engine.KismetMaterialLibrary)
  (Rendering|Material|SetScalarParameterValue "/Water/Materials/MPC/MPC_Water.MPC_Water" "ManualTime" (Utilities|Time|GetTimeSeconds) :!class /Script/Engine.KismetMaterialLibrary)
  (Rendering|Material|SetScalarParameterValue "/Water/Materials/MPC/MPC_Water.MPC_Water" "WaveCaptureSize" (Variables|Default|GetSimulation%20World%20Size) :!class /Script/Engine.KismetMaterialLibrary)
  (if (Variables|Default|GetFreeze)
    _
    (elif (Variables|Default|GetTicking)
      _))
  (CallFunction|StopTick))
